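# Lit material for heart emote effects, built for mesh particles in both Cascade and Niagara.
# Particle color RGB feeds the emissive input of the shared MF_FX_EmissiveAdjustments function.
# A red Color vector parameter and three scalar constants sit alongside it, with no visible connections to the material outputs.
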Material /Game/Effects/Fort_Effects/Materials/Emotes/M_Emote_Heart
graph {
 properties {bUsedWithMeshParticles=true, bUsedWithNiagaraMeshParticles=true, pixelDepthOffsetMode=PDOM_Legacy, usedShadingModels=Default Lit}
 nodes [
  Constant() {r=1}
  Constant() {r=0.2}
  Constant() {r=0.9}
  MaterialFunctionCall(Emissive=ParticleColor().RGB) {materialFunction=&MF_FX_EmissiveAdjustments}
  VectorParameter() {parameterName=Color, defaultValue=<rgba 0.95 0.0345 0.019 0>}
 ]
}
