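# Translucent, lit particle sprite material for the salt crystals in the Salty emote effect, with responsive AA enabled.
# A sharpened sine of particle relative time, offset by particle color green, creates twinkles that strengthen over the particle's life.
# That pulse blends LowColor to a bright HighColor through MF_FX_EmissiveAdjustments.
# A SphereMask with its radius varied by a sine of U forms a slightly irregular round shape.
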Material /Game/Effects/Fort_Effects/Materials/Emotes/M_Emote_SaltyCrystals
graph {
 properties {
  bEnableResponsiveAA=true
  bUsedWithParticleSprites=true
  blendMode=BLEND_Translucent
  pixelDepthOffsetMode=PDOM_Legacy
  usedShadingModels=Default Lit
 }
 nodes [
  n1 = Multiply(A=Add(A=ParticleRelativeTime(), B=ParticleColor().G)) {constB=12}
  n2 = Power(Base=Saturate(None=Sine(None=@n1))) {constExponent=4}
  n3 = ParticleRelativeTime()
  n4 = OneMinus(None=Power(Base=OneMinus(None=@n3)))
  MaterialFunctionCall(Emissive=LinearInterpolate(
    A=VectorParameter() {parameterName=LowColor}.RGB
    B=VectorParameter() {parameterName=HighColor, defaultValue=<rgba 6 6 6 0>}.RGB
    Alpha=Multiply(A=@n2, B=@n4)
   ) {constB=6}) {materialFunction=&MF_FX_EmissiveAdjustments}
  ParticleColor()
  n5 = ComponentMask(None=TextureCoordinate() {uTiling=1.2}) {r=true}
  SphereMask(
   A=TextureCoordinate()
   B=Constant() {r=0.5}
   Radius=LinearInterpolate(Alpha=Sine(None=@n5)) {constA=0.45, constB=0.5}
   Hardness=Constant() {r=0.5}
  )
 ]
}
