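# Translucent unlit smoky trail material for emote effects, usable on beam trails and Niagara ribbons.
# A Pythagorean function turns distance from the V midline into a rounded edge profile, combined with tiled noise on UV1 offset by dynamic SmokeTiler.
# Particle color, profile, noise and a V gradient drive emissive.
# HeightLerp erodes alpha with the noise, scaled by particle alpha, Param2, camera fade and depth fade (skipped on Low quality).
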Material /Game/Effects/Fort_Effects/Materials/Emotes/M_Emote_SimpleTrail
graph {
 properties {
  bUsedWithBeamTrails=true
  bUsedWithNiagaraRibbons=true
  blendMode=BLEND_Translucent
  pixelDepthOffsetMode=PDOM_Legacy
  shadingModel=MSM_Unlit
  usedShadingModels=Unlit
 }
 nodes [
  n1 = Subtract(B=ComponentMask(None=TextureCoordinate()) {g=true}) {constA=0.5}
  n2 = MaterialFunctionCall!(Side 1 Length=Multiply(A=Abs(None=@n1)) {constB=2}) {materialFunction=/Engine/Functions/Engine_MaterialFunctions02/Math/pythagoreanTheorum}
  n3 = DynamicParameter() {paramNames [SmokeTiler,Param2,Param3,Param4]}
  n4 = ScalarParameter() {defaultValue=0.5, parameterName=XY Shift}
  n5 = TextureSample(UVs=Add(A=TextureCoordinate() {coordinateIndex=1, uTiling=0.2, vTiling=0.02}, B=Multiply(A=@n3.SmokeTiler, B=@n4))) {texture=/Game/Effects/Fort_Effects/Textures/Tile/T_TileNoise_03}
  n6 = Multiply(A=@n2.Side 2 Length, B=@n5.G)
  n7 = ComponentMask(None=TextureCoordinate()) {g=true}
  MaterialFunctionCall(Emissive=Multiply(A=ParticleColor().RGB, B=Multiply(A=@n6, B=@n7))) {materialFunction=&MF_FX_EmissiveAdjustments}
  n8 = ParticleColor()
  n9 = DepthFade(FadeDistance=Constant() {r=100}) {fadeDistanceDefault=15}
  n10 = Multiply(
   A=QualitySwitch(Default=@n9, Low=Constant() {r=1}, High=@n9, Medium=@n9)
   B=MaterialFunctionCall(Fade Length=Constant() {r=25}) {materialFunction=&CameraDepthFade}.Result
  )
  Multiply(
   A=MaterialFunctionCall(Transition Phase=Multiply(A=@n2.Side 2 Length, B=@n2.Side 2 Length), Height Texture=LinearInterpolate(Alpha=@n5.G) {constA=-1}) {materialFunction=&HeightLerp}.Alpha
   B=Multiply(A=Multiply(A=@n8.A, B=@n10), B=DynamicParameter() {paramNames [SmokeTiler,Param2,Param3,Param4]}.Param2)
  )
 ]
}
