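# Translucent unlit particle material for twinkling emote sparkles, usable on Niagara and particle sprites.
# Particle color is brightened by a clamped time sine at per-particle random speed, slightly hue-shifted at random, blended against plain color by a parameter, then passed to emissive adjustments.
# A separate branch multiplies lens-flare green, particle random and particle alpha.
# Another multiplies a randomized ScalingSpeed sine by component pivot location, with no visible destination.
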Material /Game/Effects/Fort_Effects/Materials/Emotes/M_Emote_Sparkles
graph {
 properties {
  bEnableResponsiveAA=true
  bUsedWithNiagaraSprites=true
  bUsedWithParticleSprites=true
  blendMode=BLEND_Translucent
  opacityMaskClipValue=0.01
  pixelDepthOffsetMode=PDOM_Legacy
  shadingModel=MSM_Unlit
  usedShadingModels=Unlit
 }
 nodes [
  n1 = ParticleColor()
  n2 = ParticleRandom()
  n3 = Multiply(A=@n2) {constB=-0.04}
  n4 = MaterialFunctionCall(Speed=Multiply(A=@n2, B=ScalarParameter() {defaultValue=3, parameterName=TwinkleSpeed})) {materialFunction=&DebugTimeSine}
  n5 = Multiply(A=Clamp(None=Multiply(A=@n4.Result) {constB=5}) {maxDefault=5, minDefault=0.25}, B=@n1.RGB)
  MaterialFunctionCall(Emissive=LinearInterpolate(
    A=@n1.RGB
    B=MaterialFunctionCall(Hue Shift Percentage=@n3, Texture=@n5) {materialFunction=&HueShift}.Result
    Alpha=ScalarParameter() {defaultValue=1, parameterName=Use Particle Random Color}
   )) {materialFunction=&MF_FX_EmissiveAdjustments}
  Multiply(A=Multiply(A=TextureSample() {texture=/Game/Effects/Fort_Effects/Textures/T_Hanover_LensFlare}.G, B=ParticleRandom()), B=ParticleColor().A)
  n6 = ParticleRandom()
  n7 = ScalarParameter() {defaultValue=1500, parameterName=ScalingSpeed}
  Multiply(
   A=MaterialFunctionCall(Speed=Multiply(A=@n6, B=@n7)) {materialFunction=&DebugTimeSine}.Result
   B=MaterialFunctionCall() {materialFunction=&ComponentPivotLocation}.Result
  )
 ]
}
