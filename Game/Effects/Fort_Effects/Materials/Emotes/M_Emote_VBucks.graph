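# Two-sided lit material for V-Bucks bill mesh particles in emotes, with a static "Use on Pile" variant.
# The bill texture is tinted through HueShift by a per-particle ColorChoice, with optional particle color, and opacity uses depth and near-camera fades.
# Loose bills get Jiggle_WPO waves plus normal offsets masked by red and green vertex color and driven by dynamic parameters.
# The pile variant zeroes that offset and blends color by vertex red and UV.
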Material /Game/Effects/Fort_Effects/Materials/Emotes/M_Emote_VBucks
graph {
 properties {
  bUsedWithMeshParticles=true
  bUsedWithNiagaraMeshParticles=true
  pixelDepthOffsetMode=PDOM_Legacy
  translucencyLightingMode=TLM_VolumetricPerVertexDirectional
  twoSided=true
  usedShadingModels=Default Lit
 }
 nodes [
  n1 = TextureSample() {texture=/Game/Effects/Fort_Effects/Textures/Emotes/V_Buck_Bill_v2}
  LinearInterpolate(Alpha=@n1.R) {constA=0.4, constB=0.8}
  LinearInterpolate(Alpha=@n1.R) {constB=0.3}
  n2 = DepthFade(
   Opacity=MaterialFunctionCall(Distance=ScalarParameter() {defaultValue=100, parameterName=NearCameraFadeDistance}) {materialFunction=&Near_Camera_Fade}.Result
   FadeDistance=ScalarParameter() {parameterName=DF_Distance}
  ) {fadeDistanceDefault=0}
  n3 = ShadingPathSwitch(Default=@n2, Deferred=@n2, Mobile=Constant() {r=1}) {desc=Migrated from FeatureLevelSwitch}
  LinearInterpolate(A=@n3, B=Multiply(A=ParticleColor().A, B=@n3), Alpha=ScalarParameter() {parameterName=LerpInParticleAlpha, sliderMax=1})
  n4 = DynamicParameter() {defaultValue=<rgba 0 2 0 1>, paramNames [ColorChoice,WPO_Speed,WPO_Red,WPO_Green]}
  n5 = ComponentMask(None=TextureCoordinate()) {g=true}
  n6 = OneMinus(None=VertexColor().R)
  n7 = MaterialFunctionCall(
   Hue Shift Percentage=StaticSwitchParameter(True=LinearInterpolate(A=@n4.ColorChoice, B=@n5, Alpha=@n6), False=@n4.ColorChoice) {parameterName=Use on Pile}
   Texture=Constant3Vector() {constant=<rgba 1 0.8554 0.506 0>}
  ) {materialFunction=&HueShift}
  n8 = Multiply(A=@n1.RGB, B=@n7.Result)
  n9 = VertexColor()
  n10 = StaticSwitchParameter(True=LinearInterpolate(A=@n8, B=@n7.Result, Alpha=OneMinus(None=@n9.R)), False=@n8) {parameterName=Use on Pile}
  LinearInterpolate(A=@n10, B=Multiply(A=ParticleColor().RGB, B=@n10), Alpha=ScalarParameter() {parameterName="UseParticleColor(Lerp)"})
  ScalarParameter() {defaultValue=0.8, parameterName=Roughness}
  ScalarParameter() {defaultValue=1, parameterName=Spec}
  n11 = ScalarParameter() {defaultValue=0.5, parameterName=Small Wave}
  n12 = Multiply(A=ParticleRelativeTime(), B=ScalarParameter() {defaultValue=10, parameterName=WaveScale})
  n13 = VertexNormalWS()
  n14 = ScalarParameter() {defaultValue=2, parameterName=Amount_Offset_Red}
  n15 = Multiply(A=@n14) {constB=-1}
  n16 = DynamicParameter() {defaultValue=<rgba 0 2 0 1>, paramNames [ColorChoice,WPO_Speed,WPO_Red,WPO_Green]}
  n17 = VertexColor()
  n18 = ScalarParameter() {defaultValue=5, parameterName=Amount_Offset_Green}
  n19 = Multiply(A=@n18) {constB=-1}
  n20 = Add(
   A=Multiply(A=LinearInterpolate(A=@n15, B=@n14, Alpha=@n16.WPO_Red), B=@n17.R)
   B=Multiply(A=@n17.G, B=LinearInterpolate(A=@n19, B=@n18, Alpha=@n16.WPO_Green))
  )
  StaticSwitchParameter(
   True=Constant()
   False=Add(
    A=MaterialFunctionCall!(Speed=@n4.WPO_Speed, SmallWaveScle=@n11, WaveScale=@n12) {materialFunction=/Game/Effects/Fort_Effects/Materials/_Functions/Jiggle_WPO}.Result
    B=Multiply(A=@n13, B=@n20)
   )
  ) {parameterName=Use on Pile}
 ]
}
