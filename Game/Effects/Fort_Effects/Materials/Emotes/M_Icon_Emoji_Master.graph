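# Unlit masked master material for emote emoji icon particles, drawing a configurable icon texture as a flipbook.
# Frame index, frame count and a particle dynamic AnimationPhase pick the atlas cell, which is tinted by emissive and particle color, with combined alpha for masking.
# A camera-distance fade drives world-space vertex offsets: it scales a sine wiggle, shifts the icon to keep it visible up close, and pulls vertices toward the particle center to resize it.
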Material /Game/Effects/Fort_Effects/Materials/Emotes/M_Icon_Emoji_Master
graph {
 properties {
  bEnableResponsiveAA=true
  bUsedWithMeshDeformer=true
  bUsedWithNiagaraSprites=true
  bUsedWithParticleSprites=true
  bUsedWithSkeletalMesh=true
  blendMode=BLEND_Masked
  pixelDepthOffsetMode=PDOM_Legacy
  shadingModel=MSM_Unlit
  usedShadingModels=Unlit
 }
 nodes [
  n1 = Time()
  n2 = ScalarParameter() {defaultValue=1, parameterName=WiggleFrequency, sliderMax=1}
  n3 = Multiply(A=Sine(None=Multiply(A=@n1, B=@n2)), B=VectorParameter() {parameterName=WiggleWS, defaultValue=<rgba 0 0 1.5 0>}.RGB)
  n4 = MaterialFunctionCall(
   Position=ParticlePositionWS()
   Fade_Start_Distance=ScalarParameter() {defaultValue=160, parameterName=StartDistance}
   Fade_End_Distance=ScalarParameter() {parameterName=EndDistance}
  ) {materialFunction=&MF_Distance_SyncedFade_FX}
  n5 = VectorParameter() {desc="Move in WS as we get closer, helps prevent particle from getting lost", parameterName=OffsetByDistance, defaultValue=<rgba 0 0 -3 0>}
  n6 = Normalize(VectorInput=Subtract(A=ParticlePositionWS(), B=WorldPosition().XYZ))
  n7 = Multiply(A=ParticleRadius(), B=ScalarParameter() {defaultValue=1.4, parameterName=SizeByDistanceScale})
  Add(
   A=Add(A=Multiply(A=@n3, B=@n4.Result), B=LinearInterpolate(A=@n5.RGB, Alpha=@n4.Result))
   B=LinearInterpolate(A=Multiply(A=@n6, B=@n7) {constB=10}, Alpha=@n4.Result) {constB=0}
  )
  n8 = VectorParameter() {parameterName=Emissive Color, defaultValue=<rgba 1 1 1 0>}
  n9 = ParticleColor()
  n10 = ScalarParameter() {parameterName=FrameIndex}
  n11 = Multiply(A=ScalarParameter() {defaultValue=1, parameterName=Rows}, B=ScalarParameter() {defaultValue=1, parameterName=Columns})
  n12 = Divide(A=ScalarParameter() {defaultValue=1, parameterName=FrameCount}, B=@n11)
  n13 = DynamicParameter() {paramNames [PulseSpeed, AnimationPhase, Param3, Param4]}
  n14 = MaterialFunctionCall(
   "Animation  Phase (0-1)"=Add(A=Divide(A=@n10, B=@n11), B=Multiply(A=@n12, B=@n13.AnimationPhase)) {desc=Animation Offset}
   Number of Rows=ScalarParameter() {defaultValue=1, parameterName=Rows}
   Number of Columns=ScalarParameter() {defaultValue=1, parameterName=Columns}
   Texture=TextureObjectParameter() {parameterName=IconTexture, texture=/Game/Global/Textures/Default/Blanks/T_Blank_White_ZeroAlpha_sRGBColor}
  ) {materialFunction=&FlipBook}
  n15 = Multiply(A=@n8.RGB, B=Multiply(A=@n9.RGB, B=@n14.Result))
  Divide(A=@n15, B=EyeAdaptation().EyeAdaptation)
  n16 = Divide(A=@n8.RGB, B=EyeAdaptation().EyeAdaptation)
  FeatureLevelSwitch(Default=@n16, ES2=@n8.RGB, ES3_1=@n8.RGB, SM4_REMOVED=@n16, SM5=@n16, SM6=@n16)
  MaterialFunctionCall(Emissive=@n15) {materialFunction=&MF_FX_EmissiveAdjustments}
  Reroute(None=Multiply(A=@n9.A, B=@n14.Alpha))
  Time() {bOverride_Period=true, period=1}
 ]
}
