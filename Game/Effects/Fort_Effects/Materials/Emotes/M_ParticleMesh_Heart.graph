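# Translucent unlit mesh-particle material for a heart emote effect, drawing the anime flare texture tinted by particle color.
# A Sprite function in world position offset forms sprites, while an offset along vertex normals expands them.
# Vertex color red times the dynamic RandomMult varies sprite scale, expansion and opacity.
# Emissive is remapped to HDR and adjusted; opacity includes a near-camera fade, skipped at low quality.
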Material /Game/Effects/Fort_Effects/Materials/Emotes/M_ParticleMesh_Heart
graph {
 properties {
  bUsedWithMeshParticles=true
  bUsedWithNiagaraMeshParticles=true
  blendMode=BLEND_Translucent
  pixelDepthOffsetMode=PDOM_Legacy
  shadingModel=MSM_Unlit
  usedShadingModels=Unlit
 }
 nodes [
  n1 = DynamicParameter() {paramNames [Scale,Expand,RandomMult,Param4]}
  n2 = Multiply(A=VertexColor().R, B=@n1.RandomMult)
  Add(
   A=Multiply(A=VertexNormalWS(), B=Multiply(A=@n2, B=@n1.Expand))
   B=MaterialFunctionCall!(Center Location=WorldPosition().XYZ, 0-1 UVs, XY scale=Multiply(A=@n2, B=@n1.Scale), OverridePivot, "Normals (Optional)") {materialFunction=&Sprite}.World position offset
  )
  Constant3Vector() {constant=<rgba 20 6.14 17.0978 1>}
  n3 = ParticleColor()
  n4 = TextureSample() {texture=/Game/Effects/Fort_Effects/Textures/Flares/T_Anime_Flare}
  n5 = Multiply(A=@n3.RGB, B=@n4.G)
  MaterialFunctionCall(Emissive=MaterialFunctionCall(LDR Input=Clamp(None=@n5), HDR Tint=@n5) {materialFunction=&DeriveHDRfromLDR}.RemappedRange) {materialFunction=&MF_FX_EmissiveAdjustments}
  n6 = MaterialFunctionCall(Distance=ScalarParameter() {defaultValue=200, parameterName=CameraDistanceFade}) {materialFunction=&Near_Camera_Fade}
  Multiply(
   A=Multiply(A=Multiply(A=@n4.G, B=@n3.A), B=@n2)
   B=QualitySwitch(Default=@n6.Result, Low=Constant() {r=1}, High=@n6.Result, Medium=@n6.Result)
  )
 ]
}
