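# Unlit translucent two-sided rainbow material for emote effects, flagged for beam trails and Niagara ribbons.
# A rainbow gradient, sampled on swizzled UVs offset by ColorScroll, is scaled by time-of-day brightness and particle color, then remapped to HDR emissive.
# Opacity multiplies particle alpha by four-sided UV edge falloffs with adjustable contrast, depth fade and camera depth fade.
# The mobile path skips those falloffs and fades, using only particle alpha.
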Material /Game/Effects/Fort_Effects/Materials/Emotes/M_RainbowHDR
graph {
 properties {
  bUsedWithBeamTrails=true
  bUsedWithNiagaraRibbons=true
  blendMode=BLEND_Translucent
  pixelDepthOffsetMode=PDOM_Legacy
  shadingModel=MSM_Unlit
  twoSided=true
  usedShadingModels=Unlit
 }
 nodes [
  n1 = TextureCoordinate()
  n2 = MaterialFunctionCall(Float2=@n1) {materialFunction=&BreakOutFloat2Components}
  n3 = DotProduct(A=VertexNormalWS(), B=CameraVectorWS())
  LinearInterpolate(
   A=@n1
   B=AppendVector(A=OneMinus(None=@n2.R), B=@n2.G)
   Alpha=MaterialFunctionCall!(Mask Offset Value, Number Before White Result, Gradient=Clamp(None=@n3)) {materialFunction=/Engine/Functions/Engine_MaterialFunctions02/ValueStep}.Result
  )
  n4 = ScalarParameter() {defaultValue=1, parameterName=Brightness_Morning}
  n5 = ScalarParameter() {defaultValue=1.4, parameterName=Brightness_Day}
  n6 = ScalarParameter() {defaultValue=1, parameterName=Brightness_Evening}
  n7 = ScalarParameter() {defaultValue=0.65, parameterName=Brightness_Night}
  n8 = Add(
   A=MaterialFunctionCall() {materialFunction=&Swizzle}.YX
   B=ScalarParameter() {defaultValue=0.2, parameterName=ColorScroll}
  )
  n9 = ParticleColor()
  n10 = Multiply(
   A=Multiply(
    A=MaterialFunctionCall(Color_Morning=@n4, Color_Day=@n5, Color_Evening=@n6, Color_Night=@n7) {materialFunction=&Color_Time_of_Day}.Result
    B=TextureSample(UVs=@n8) {texture=/Game/Global/Textures/Default/Masks_and_Gradients/T_Grad_Rainbow}.RGB
   )
   B=@n9.RGB
  )
  MaterialFunctionCall(Emissive=MaterialFunctionCall(LDR Input=Clamp(None=@n10), HDR Tint=@n10) {materialFunction=&DeriveHDRfromLDR}.RemappedRange) {materialFunction=&MF_FX_EmissiveAdjustments}
  n11 = ComponentMask(None=TextureCoordinate()) {r=true}
  n12 = ScalarParameter() {defaultValue=1, group=Opacity, parameterName=Left-Right Contrast, sortPriority=2}
  n13 = OneMinus(None=ComponentMask(None=TextureCoordinate()) {r=true})
  n14 = ScalarParameter() {defaultValue=1, group=Opacity, parameterName=Right-Left Contrast, sortPriority=3}
  n15 = ComponentMask(None=TextureCoordinate()) {g=true}
  n16 = ScalarParameter() {defaultValue=1, group=Opacity, parameterName=Top-Bottom Contrast, sortPriority=4}
  n17 = OneMinus(None=ComponentMask(None=TextureCoordinate()) {g=true})
  n18 = ScalarParameter() {defaultValue=1, group=Opacity, parameterName=Bottom-Top Contrast, sortPriority=5}
  n19 = Multiply(
   A=Multiply(A=Power(Base=@n11, Exp=@n12), B=Power(Base=@n13, Exp=@n14))
   B=Multiply(A=Power(Base=@n15, Exp=@n16), B=Power(Base=@n17, Exp=@n18))
  )
  n20 = ScalarParameter() {defaultValue=100, group=Opacity, parameterName=Opacity_Intensity_Compensator, sortPriority=1}
  n21 = DepthFade(Opacity=Clamp(None=Multiply(A=@n19, B=@n20))) {fadeDistanceDefault=5}
  n22 = MaterialFunctionCall() {materialFunction=&CameraDepthFade}
  n23 = Multiply(A=Multiply(A=@n9.A, B=Multiply(A=@n21, B=@n22.Result)))
  ShadingPathSwitch(Default=@n23, Deferred=@n23, Mobile=Multiply(A=@n9.A)) {desc=Migrated from FeatureLevelSwitch}
 ]
}
